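# Looping stereo MetaSound source for the mounted turret's rotation sound, driven by horizontal and vertical aim deltas.
# On play, two heavy turn-loop wave players start together and feed a two-channel stereo mixer.
# Vertical aim delta sets the first loop's gain and horizontal the second's, each smoothed by InterpTo so volume follows aim movement.
# Concurrency override allows one instance per owner, with close-range weapon attenuation.
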MetaSoundSource /Game/Sounds/Fort_Weapons/Turret/MSS_Turret_Mounted_Auto_Rotate
inputs [
 HorizontalAimDelta {type=Float}
 UE.Source.OnPlay {type=Trigger}
 VerticalAimDelta {type=Float}
]
interfaces [UE.OutputFormat.Stereo:1.0, UE.Source:1.1]
outputs [
 "UE.OutputFormat.Stereo.Audio:0" {type=Audio}
 "UE.OutputFormat.Stereo.Audio:1" {type=Audio}
]
pages [Default {
  input_defaults {HorizontalAimDelta=0, UE.Source.OnPlay=null, VerticalAimDelta=0.048}
  nodes [
   WavePlayer_0 = UE.Wave Player:Stereo(Play=^UE.Source.OnPlay) {also {defaults {Loop="<MetaSoundBuilderResult.SUCCEEDED: 0>", Wave Asset=&/WPN_Turret_Mounted_Heavy_Turn_Loop_01}}}
   WavePlayer_1 = UE.Wave Player:Stereo(Play=^UE.Source.OnPlay) {also {defaults {Loop="<MetaSoundBuilderResult.SUCCEEDED: 0>", Wave Asset=&/WPN_Turret_Mounted_Heavy_Turn_Loop_02}}}
   AudioMixerAudioMixerStereo2_0 = "AudioMixer.Audio Mixer (Stereo, 2)"(
    Gain 0=UE.InterpTo:Audio(Target=^VerticalAimDelta) {also {defaults {Interp Time=0.35}}}.Value
    Gain 1=UE.InterpTo:Audio(Target=^HorizontalAimDelta) {also {defaults {Interp Time=0.35}}}.Value
    In 0 L=@WavePlayer_0.Out Left
    In 0 R=@WavePlayer_0.Out Right
    In 1 L=@WavePlayer_1.Out Left
    In 1 R=@WavePlayer_1.Out Right
   )
  ]
  links [
   {dst="@UE.OutputFormat.Stereo.Audio:0", src=AudioMixerAudioMixerStereo2_0.Out L}
   {dst="@UE.OutputFormat.Stereo.Audio:1", src=AudioMixerAudioMixerStereo2_0.Out R}
  ]
 }]
properties {
 assetImportData=&self:AssetImportData
 attenuationSettings=&Fort_Weapon_Projectile_Close_Athena_Att
 bLooping=true
 bOverrideConcurrency=true
 duration=10000
 maxDistance=4950
 numChannels=2
 outputFormat=Stereo
 sampleRate=48000
 soundClassObject=&Fort_Traps
 volume=0.5
 concurrencyOverrides {bLimitToOwner=true, maxCount=1}
}
