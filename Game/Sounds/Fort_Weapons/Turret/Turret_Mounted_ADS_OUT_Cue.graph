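# Sound cue for the foley played when leaving aim-down-sights on a mounted turret.
# A Random node picks one of three equally weighted ADS-out wave variations and feeds a Modulator at the root.
# It plays quietly through the weapon foley sound class, using shared weapon reload attenuation.
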SoundCue /Game/Sounds/Fort_Weapons/Turret/Turret_Mounted_ADS_OUT_Cue
root=Modulator_0
nodes [
 {id=Modulator_0, children [Random_1]}
 {id=Random_1, children [WavePlayer_0,WavePlayer_3,WavePlayer_4], properties {weights [1,1,1]}}
 {id=WavePlayer_0, properties {soundWaveAssetPtr=&/WPN_Turret_Mounted_ADS_Out_02}}
 {id=WavePlayer_3, properties {soundWaveAssetPtr=&/WPN_Turret_Mounted_ADS_Out_03}}
 {id=WavePlayer_4, properties {soundWaveAssetPtr=&/WPN_Turret_Mounted_ADS_Out_04}}
]
properties {
 attenuationSettings=&Fort_Weapon_Reload_Att
 maxDistance=2250
 soundClassObject=&Fort_WeaponFoley
 volumeMultiplier=0.2
}
