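# Sound cue for the enter sound of the mounted auto turret weapon.
# A single wave player feeds a modulator left at default settings, with the cue's volume boosted.
# It plays in the traps sound class with close-range weapon projectile attenuation.
# A concurrency override allows only one instance at a time and rejects new plays while it is active.
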SoundCue /Game/Sounds/Fort_Weapons/Turret/Turret_Mounted_Auto_Enter_Cue
root=Modulator_1
nodes [
 {id=Modulator_1, children [WavePlayer_0]}
 {id=WavePlayer_0, properties {soundWaveAssetPtr=&/WPN_Turret_Mounted_Auto_Enter_01}}
]
properties {
 attenuationSettings=&Fort_Weapon_Projectile_Close_Athena_Att
 maxDistance=4950
 soundClassObject=&Fort_Traps
 volumeMultiplier=2
 concurrencyOverrides {maxCount=1, resolutionRule=PreventNew}
}
