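# First-person sound cue for the mounted turret's automatic fire, mixing gunshots with shell sounds.
# A quality-level switch gives the first two slots a random pick of five fire waves and the last two a single fixed wave.
# Two random shell layers draw on eight variations, one immediate and one delayed half a second.
# A modulator slightly varies volume, with first-person gunfire attenuation and concurrency applied.
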SoundCue /Game/Sounds/Fort_Weapons/Turret/Turret_Mounted_Auto_Fire_1P_Cue
root=Modulator_0
nodes [
 {id=Delay_2, children [Random_2], properties {delayMax=0.5, delayMin=0.5}}
 {id=Mixer_0, children [QualityLevel_0, Random_1, Delay_2], properties {inputVolume [1,1,1]}}
 {id=Modulator_0, children [Mixer_0], properties {volumeMax=1, volumeMin=0.9}}
 {id=QualityLevel_0, children [Random_0,Random_0,WavePlayer_5,WavePlayer_5]}
 {id=Random_0, children [WavePlayer_0,WavePlayer_1,WavePlayer_2,WavePlayer_3,WavePlayer_4], properties {weights [1,1,1,1,1]}}
 {
  id=Random_1
  children [WavePlayer_6, WavePlayer_7, WavePlayer_8, WavePlayer_9, WavePlayer_10, WavePlayer_11, WavePlayer_12, WavePlayer_13]
  properties {weights [1,1,1,1,1,1,1,1]}
 }
 {
  id=Random_2
  children [WavePlayer_14, WavePlayer_15, WavePlayer_19, WavePlayer_20, WavePlayer_21, WavePlayer_22, WavePlayer_23, WavePlayer_24]
  properties {weights [1,1,1,1,1,1,1,1]}
 }
 {id=WavePlayer_0, properties {soundWaveAssetPtr=&/WPN_Turret_Mounted_Auto_Fire_1P_05}}
 {id=WavePlayer_1, properties {soundWaveAssetPtr=&/WPN_Turret_Mounted_Auto_Fire_1P_04}}
 {id=WavePlayer_10, properties {soundWaveAssetPtr=&/WPN_Turret_Mounted_Shell_01_05}}
 {id=WavePlayer_11, properties {soundWaveAssetPtr=&/WPN_Turret_Mounted_Shell_01_06}}
 {id=WavePlayer_12, properties {soundWaveAssetPtr=&/WPN_Turret_Mounted_Shell_01_07}}
 {id=WavePlayer_13, properties {soundWaveAssetPtr=&/WPN_Turret_Mounted_Shell_01_08}}
 {id=WavePlayer_14, properties {soundWaveAssetPtr=&/WPN_Turret_Mounted_Shell_01_01}}
 {id=WavePlayer_15, properties {soundWaveAssetPtr=&/WPN_Turret_Mounted_Shell_01_02}}
 {id=WavePlayer_19, properties {soundWaveAssetPtr=&/WPN_Turret_Mounted_Shell_01_03}}
 {id=WavePlayer_2, properties {soundWaveAssetPtr=&/WPN_Turret_Mounted_Auto_Fire_1P_03}}
 {id=WavePlayer_20, properties {soundWaveAssetPtr=&/WPN_Turret_Mounted_Shell_01_04}}
 {id=WavePlayer_21, properties {soundWaveAssetPtr=&WPN_Turret_Mounted_Shell_01_05}}
 {id=WavePlayer_22, properties {soundWaveAssetPtr=&WPN_Turret_Mounted_Shell_01_06}}
 {id=WavePlayer_23, properties {soundWaveAssetPtr=&WPN_Turret_Mounted_Shell_01_07}}
 {id=WavePlayer_24, properties {soundWaveAssetPtr=&WPN_Turret_Mounted_Shell_01_08}}
 {id=WavePlayer_3, properties {soundWaveAssetPtr=&/WPN_Turret_Mounted_Auto_Fire_1P_02}}
 {id=WavePlayer_4, properties {soundWaveAssetPtr=&/WPN_Turret_Mounted_Auto_Fire_1P_01}}
 {id=WavePlayer_5, properties {soundWaveAssetPtr=&WPN_Turret_Mounted_Auto_Fire_1P_01}}
 {id=WavePlayer_6, properties {soundWaveAssetPtr=&WPN_Turret_Mounted_Shell_01_01}}
 {id=WavePlayer_7, properties {soundWaveAssetPtr=&WPN_Turret_Mounted_Shell_01_02}}
 {id=WavePlayer_8, properties {soundWaveAssetPtr=&WPN_Turret_Mounted_Shell_01_03}}
 {id=WavePlayer_9, properties {soundWaveAssetPtr=&WPN_Turret_Mounted_Shell_01_04}}
]
properties {
 attenuationSettings=&Fort_Weapon_Projectile_FirstPerson_Athena_Att
 maxDistance=2097152
 soundClassObject=&Fort_WeaponRanged_Fire_1P_PSM
 volumeMultiplier=0.85
 attenuationOverrides {distanceAlgorithm=NaturalSound, falloffDistance=4000, attenuationShapeExtents {x=1000}}
 concurrencyOverrides {maxCount=3}
 concurrencySet [&Weapon_Gunfire_FirstPerson_Conc]
}
