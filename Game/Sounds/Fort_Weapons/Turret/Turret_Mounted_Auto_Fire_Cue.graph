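# Sound cue for the mounted turret's automatic fire, layering close, medium and distant shot sounds.
# Each layer picks a random wave, passes through a modulator, its own distance attenuation and a sound class override, then mixes into the root.
# The close layer's quality switch swaps the five-wave random pool for one fixed wave on two of its four levels.
# The cue overrides concurrency to cap simultaneous plays at five.
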SoundCue /Game/Sounds/Fort_Weapons/Turret/Turret_Mounted_Auto_Fire_Cue
root=Mixer_15
nodes [
 {
  id=Attenuation_80
  children [Modulator_0]
  properties {attenuationSettings=/Game/Sounds/Attenuation/Fort_Weapon_Projectile_Close_Athena_Att_NonSpac}
 }
 {
  id=Attenuation_81
  children [Modulator_13]
  properties {attenuationSettings=&Fort_Weapon_Projectile_Medium_Mono_Athena_Att}
 }
 {
  id=Attenuation_82
  children [Modulator_1]
  properties {attenuationSettings=&Fort_Weapon_Projectile_Distant_Mono_Athena_Att}
 }
 {id=Mixer_0, children [QualityLevel_0, null, null], properties {inputVolume [1,0.2,0.2]}}
 {id=Mixer_1, children [SoundClass_56, SoundClass_0], properties {inputVolume [1,1]}}
 {id=Mixer_15, children [SoundClass_55, Mixer_1], properties {inputVolume [0.7,1]}}
 {id=Modulator_0, children [Mixer_0], properties {volumeMax=1, volumeMin=0.9}}
 {id=Modulator_1, children [Random_2], properties {volumeMax=0.85, volumeMin=0.8}}
 {id=Modulator_13, children [Random_1], properties {volumeMin=1}}
 {id=QualityLevel_0, children [Random_0,Random_0,WavePlayer_5,WavePlayer_5]}
 {id=Random_0, children [WavePlayer_0,WavePlayer_1,WavePlayer_2,WavePlayer_3,WavePlayer_4], properties {weights [1,1,1,1,1]}}
 {id=Random_1, children [WavePlayer_11, WavePlayer_12, WavePlayer_13, WavePlayer_14, WavePlayer_15], properties {weights [1,1,1,1,1]}}
 {id=Random_2, children [WavePlayer_19, WavePlayer_20, WavePlayer_21, WavePlayer_22], properties {weights [1,1,1,1]}}
 {
  id=SoundClass_0
  children [Attenuation_82]
  properties {soundClassOverride=&Fort_WeaponRanged_Fire_Distant}
 }
 {
  id=SoundClass_55
  children [Attenuation_80]
  properties {soundClassOverride=&Fort_WeaponRanged_Fire_3P_PSM}
 }
 {
  id=SoundClass_56
  children [Attenuation_81]
  properties {soundClassOverride=&Fort_WeaponRanged_Fire_Distant}
 }
 {id=WavePlayer_0, properties {soundWaveAssetPtr=&/WPN_Turret_Mounted_Auto_Fire_Close_01}}
 {id=WavePlayer_1, properties {soundWaveAssetPtr=&/WPN_Turret_Mounted_Auto_Fire_Close_02}}
 {id=WavePlayer_11, properties {soundWaveAssetPtr=&/WPN_Turret_Mounted_Auto_Fire_3P_Medium_04}}
 {id=WavePlayer_12, properties {soundWaveAssetPtr=&/WPN_Turret_Mounted_Auto_Fire_3P_Medium_02}}
 {id=WavePlayer_13, properties {soundWaveAssetPtr=&/WPN_Turret_Mounted_Auto_Fire_3P_Medium_03}}
 {id=WavePlayer_14, properties {soundWaveAssetPtr=&/WPN_Turret_Mounted_Auto_Fire_3P_Medium_05}}
 {id=WavePlayer_15, properties {soundWaveAssetPtr=&/WPN_Turret_Mounted_Auto_Fire_3P_Medium_01}}
 {id=WavePlayer_19, properties {soundWaveAssetPtr=&/WPN_Turret_Mounted_Auto_Fire_3P_Distant_01}}
 {id=WavePlayer_2, properties {soundWaveAssetPtr=&/WPN_Turret_Mounted_Auto_Fire_Close_03}}
 {id=WavePlayer_20, properties {soundWaveAssetPtr=&/WPN_Turret_Mounted_Auto_Fire_3P_Distant_02}}
 {id=WavePlayer_21, properties {soundWaveAssetPtr=&/WPN_Turret_Mounted_Auto_Fire_3P_Distant_03}}
 {id=WavePlayer_22, properties {soundWaveAssetPtr=&/WPN_Turret_Mounted_Auto_Fire_3P_Distant_04}}
 {id=WavePlayer_3, properties {soundWaveAssetPtr=&/WPN_Turret_Mounted_Auto_Fire_Close_04}}
 {id=WavePlayer_4, properties {soundWaveAssetPtr=&/WPN_Turret_Mounted_Auto_Fire_Close_05}}
 {id=WavePlayer_5, properties {soundWaveAssetPtr=&WPN_Turret_Mounted_Auto_Fire_Close_01}}
]
properties {
 bOverrideConcurrency=true
 maxDistance=27500
 volumeMultiplier=0.85
 attenuationOverrides {distanceAlgorithm=NaturalSound, falloffDistance=4000, attenuationShapeExtents {x=1000}}
 concurrencyOverrides {maxCount=5}
}
