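# Sound cue for the overheat sound of the auto mounted turret weapon.
# A single overheat wave player feeds a Modulator that randomizes pitch slightly upward.
# It uses the weapon reload attenuation asset with the third-person ranged-fire sound class.
# A concurrency override is set to prevent new instances when its limit is reached.
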SoundCue /Game/Sounds/Fort_Weapons/Turret/Turret_Mounted_Auto_Overheat_Cue
root=Modulator_0
nodes [
 {id=Modulator_0, children [WavePlayer_1], properties {pitchMax=1.2, pitchMin=1}}
 {id=WavePlayer_1, properties {soundWaveAssetPtr=&/WPN_Turret_Mounted_Auto_Overheat_01}}
]
properties {
 attenuationSettings=&Fort_Weapon_Reload_Att
 bOverrideConcurrency=true
 maxDistance=2250
 soundClassObject=&Fort_WeaponRanged_Fire_3P_PSM
 volumeMultiplier=1
 attenuationOverrides {distanceAlgorithm=NaturalSound, falloffDistance=4000, attenuationShapeExtents {x=1000}}
 concurrencyOverrides {resolutionRule=PreventNew}
}
