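# First-person fire Sound Cue for the mounted heavy turret, layering the shot with its mechanical follow-through.
# A random close shot and a random mech sound play together immediately, followed by quieter delayed layers for a random reload and then a pitched-down random shell drop.
# Slight volume modulation varies the shot and reload layers.
# It uses first-person weapon attenuation, sound class and gunfire concurrency, and has the debug flag enabled.
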SoundCue /Game/Sounds/Fort_Weapons/Turret/Turret_Mounted_HV_Fire_1P_Cue
root=Mixer_0
nodes [
 {id=Delay_0, children [Random_2], properties {delayMax=0.4, delayMin=0.4}}
 {id=Delay_1, children [Random_3], properties {delayMax=1.4, delayMin=1.4}}
 {id=Mixer_0, children [Modulator_0,Modulator_1,Modulator_2], properties {inputVolume [1,0.4,0.25]}}
 {id=Mixer_1, children [Random_4,Random_5], properties {inputVolume [1,1]}}
 {id=Modulator_0, children [Mixer_1], properties {volumeMax=1.1, volumeMin=1}}
 {id=Modulator_1, children [Delay_0], properties {volumeMax=1.1, volumeMin=1}}
 {id=Modulator_2, children [Delay_1], properties {pitchMax=0.9, pitchMin=0.9}}
 {id=Random_2, children [WavePlayer_5, WavePlayer_13, WavePlayer_14, WavePlayer_15], properties {weights [1,1,1,1]}}
 {
  id=Random_3
  children [WavePlayer_16, WavePlayer_17, WavePlayer_18, WavePlayer_19, WavePlayer_20, WavePlayer_21, WavePlayer_22]
  properties {weights [1,1,1,1,1,1,1]}
 }
 {id=Random_4, children [WavePlayer_1,WavePlayer_6,WavePlayer_7,WavePlayer_8], properties {weights [1,1,1,1]}}
 {id=Random_5, children [WavePlayer_9, WavePlayer_10, WavePlayer_11, WavePlayer_12], properties {weights [1,1,1,1]}}
 {id=WavePlayer_1, properties {soundWaveAssetPtr=&/WPN_Turret_Mounted_Heavy_Fire_1P_Close_01}}
 {id=WavePlayer_10, properties {soundWaveAssetPtr=&/WPN_Turret_Mounted_Heavy_Fire_1P_Mech_02}}
 {id=WavePlayer_11, properties {soundWaveAssetPtr=&/WPN_Turret_Mounted_Heavy_Fire_1P_Mech_03}}
 {id=WavePlayer_12, properties {soundWaveAssetPtr=&/WPN_Turret_Mounted_Heavy_Fire_1P_Mech_04}}
 {id=WavePlayer_13, properties {soundWaveAssetPtr=&/WPN_Turret_Mounted_Heavy_Reload_02}}
 {id=WavePlayer_14, properties {soundWaveAssetPtr=&/WPN_Turret_Mounted_Heavy_Reload_03}}
 {id=WavePlayer_15, properties {soundWaveAssetPtr=&/WPN_Turret_Mounted_Heavy_Reload_04}}
 {id=WavePlayer_16, properties {soundWaveAssetPtr=&/WPN_Turret_Mounted_Heavy_Shell_Drop_01}}
 {id=WavePlayer_17, properties {soundWaveAssetPtr=&/WPN_Turret_Mounted_Heavy_Shell_Drop_02}}
 {id=WavePlayer_18, properties {soundWaveAssetPtr=&/WPN_Turret_Mounted_Heavy_Shell_Drop_03}}
 {id=WavePlayer_19, properties {soundWaveAssetPtr=&/WPN_Turret_Mounted_Heavy_Shell_Drop_04}}
 {id=WavePlayer_20, properties {soundWaveAssetPtr=&/WPN_Turret_Mounted_Heavy_Shell_Drop_05}}
 {id=WavePlayer_21, properties {soundWaveAssetPtr=&/WPN_Turret_Mounted_Heavy_Shell_Drop_06}}
 {id=WavePlayer_22, properties {soundWaveAssetPtr=&/WPN_Turret_Mounted_Heavy_Shell_Drop_07}}
 {id=WavePlayer_5, properties {soundWaveAssetPtr=&/WPN_Turret_Mounted_Heavy_Reload_01}}
 {id=WavePlayer_6, properties {soundWaveAssetPtr=&/WPN_Turret_Mounted_Heavy_Fire_1P_Close_02}}
 {id=WavePlayer_7, properties {soundWaveAssetPtr=&/WPN_Turret_Mounted_Heavy_Fire_1P_Close_03}}
 {id=WavePlayer_8, properties {soundWaveAssetPtr=&/WPN_Turret_Mounted_Heavy_Fire_1P_Close_04}}
 {id=WavePlayer_9, properties {soundWaveAssetPtr=&/WPN_Turret_Mounted_Heavy_Fire_1P_Mech_01}}
]
properties {
 attenuationSettings=&Fort_Weapon_Projectile_FirstPerson_Athena_Att
 bDebug=true
 maxDistance=2097152
 soundClassObject=&Fort_WeaponRanged_Fire_1P_PSM
 volumeMultiplier=1
 attenuationOverrides {reverbWetLevelMax=0.7, reverbWetLevelMin=0.2}
 concurrencySet [&Weapon_Gunfire_FirstPerson_Conc]
}
