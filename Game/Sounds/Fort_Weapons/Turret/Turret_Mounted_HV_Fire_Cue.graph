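# Sound cue for the mounted turret's HV weapon fire, layering close, medium and distant shots by range.
# Each layer randomizes wave variations through its own attenuation asset, with a third-person sound class up close and a distant-fire class on the other two.
# The close layer adds a delayed, quieter reload sound; modulators adjust volume and distant pitch.
# A concurrency override refuses new instances at the limit.
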SoundCue /Game/Sounds/Fort_Weapons/Turret/Turret_Mounted_HV_Fire_Cue
root=Mixer_3
nodes [
 {id=Attenuation_10, children [Random_1], properties {attenuationSettings=/Game/Sounds/Attenuation/Fort_Weapon_Projectile_Medium_StW_Att}}
 {
  id=Attenuation_11
  children [Modulator_7]
  properties {attenuationSettings=&Fort_Weapon_Projectile_Distant_Mono_Athena_Att}
 }
 {id=Attenuation_9, children [Mixer_2], properties {attenuationSettings=&Fort_Weapon_Projectile_Close_Athena_Att}}
 {id=Delay_0, children [WavePlayer_5], properties {delayMax=0.95, delayMin=0.95}}
 {id=Mixer_0, children [Modulator_0,SoundClass_1], properties {inputVolume [2,1]}}
 {id=Mixer_2, children [Modulator_6,Modulator_1], properties {inputVolume [1,0.5]}}
 {id=Mixer_3, children [SoundClass_7,Mixer_0], properties {inputVolume [1,1]}}
 {id=Modulator_0, children [SoundClass_0], properties {volumeMax=1, volumeMin=1}}
 {id=Modulator_1, children [Delay_0], properties {volumeMax=1.1, volumeMin=1}}
 {id=Modulator_6, children [Random_0], properties {volumeMax=1.25, volumeMin=1.2}}
 {id=Modulator_7, children [Random_2], properties {pitchMax=1.1}}
 {id=Random_0, children [WavePlayer_0, WavePlayer_1, WavePlayer_2, WavePlayer_3, WavePlayer_18], properties {weights [1,1,1,1,1]}}
 {id=Random_1, children [WavePlayer_4,WavePlayer_6,WavePlayer_7,WavePlayer_8], properties {weights [1,1,1,1]}}
 {id=Random_2, children [WavePlayer_9, WavePlayer_10, WavePlayer_11, WavePlayer_12], properties {weights [1,1,1,1]}}
 {
  id=SoundClass_0
  children [Attenuation_10]
  properties {soundClassOverride=&Fort_WeaponRanged_Fire_Distant}
 }
 {
  id=SoundClass_1
  children [Attenuation_11]
  properties {soundClassOverride=&Fort_WeaponRanged_Fire_Distant}
 }
 {
  id=SoundClass_7
  children [Attenuation_9]
  properties {soundClassOverride=&Fort_WeaponRanged_Fire_3P_PSM}
 }
 {id=WavePlayer_0, properties {soundWaveAssetPtr=&/WPN_Turret_Mounted_Auto_Fire_Close_01}}
 {id=WavePlayer_1, properties {soundWaveAssetPtr=&/WPN_Turret_Mounted_Auto_Fire_Close_02}}
 {id=WavePlayer_10, properties {soundWaveAssetPtr=&/WPN_Turret_Mounted_Heavy_Fire_3P_Dist_04}}
 {id=WavePlayer_11, properties {soundWaveAssetPtr=&/WPN_Turret_Mounted_Heavy_Fire_3P_Dist_03}}
 {id=WavePlayer_12, properties {soundWaveAssetPtr=&/WPN_Turret_Mounted_Heavy_Fire_3P_Dist_02}}
 {id=WavePlayer_18, properties {soundWaveAssetPtr=&/WPN_Turret_Mounted_Auto_Fire_Close_05}}
 {id=WavePlayer_2, properties {soundWaveAssetPtr=&/WPN_Turret_Mounted_Auto_Fire_Close_03}}
 {id=WavePlayer_3, properties {soundWaveAssetPtr=&/WPN_Turret_Mounted_Auto_Fire_Close_04}}
 {id=WavePlayer_4, properties {soundWaveAssetPtr=&/WPN_Turret_Mounted_Heavy_Fire_3P_Med_01}}
 {id=WavePlayer_5, properties {soundWaveAssetPtr=&/WPN_Turret_Mounted_Heavy_Reload_01}}
 {id=WavePlayer_6, properties {soundWaveAssetPtr=&/WPN_Turret_Mounted_Heavy_Fire_3P_Med_04}}
 {id=WavePlayer_7, properties {soundWaveAssetPtr=&/WPN_Turret_Mounted_Heavy_Fire_3P_Med_03}}
 {id=WavePlayer_8, properties {soundWaveAssetPtr=&/WPN_Turret_Mounted_Heavy_Fire_3P_Med_02}}
 {id=WavePlayer_9, properties {soundWaveAssetPtr=&/WPN_Turret_Mounted_Heavy_Fire_3P_Dist_01}}
]
properties {
 bOverrideConcurrency=true
 maxDistance=27500
 volumeMultiplier=1
 attenuationOverrides {falloffDistance=5000, reverbWetLevelMax=0.7, reverbWetLevelMin=0.2, attenuationShapeExtents {x=1000}}
 concurrencyOverrides {resolutionRule=PreventNew}
}
