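# Sound cue for the overheat sound of the mounted heavy turret weapon.
# A single wave player plays one overheat wave directly from the root at half volume, with no variation.
# It uses close weapon-projectile attenuation and the third-person ranged weapon fire sound class.
# A concurrency override prevents new instances once the limit is reached.
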SoundCue /Game/Sounds/Fort_Weapons/Turret/Turret_Mounted_HV_Overheat_Cue
root=WavePlayer_0
nodes [{id=WavePlayer_0, properties {soundWaveAssetPtr=&/WPN_Turret_Mounted_Heavy_Overheat_01}}]
properties {
 attenuationSettings=&Fort_Weapon_Projectile_Close_Athena_Att
 bOverrideConcurrency=true
 maxDistance=4950
 soundClassObject=&Fort_WeaponRanged_Fire_3P_PSM
 volumeMultiplier=0.5
 attenuationOverrides {distanceAlgorithm=NaturalSound, falloffDistance=4000, attenuationShapeExtents {x=1000}}
 concurrencyOverrides {resolutionRule=PreventNew}
}
