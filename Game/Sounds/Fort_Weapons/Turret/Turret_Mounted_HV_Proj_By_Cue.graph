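# SoundCue for projectile fly-bys from a mounted turret's heavy rounds, classed as ranged weapon audio.
# A Random node picks one of five equally weighted whiz-by waves, and a Modulator varies its pitch, with volume boosted at cue level.
# Overridden attenuation disables distance volume falloff but keeps a distance-based low-pass filter, so far passes sound duller rather than quieter.
# A bullet whiz-by concurrency rule governs simultaneous instances.
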SoundCue /Game/Sounds/Fort_Weapons/Turret/Turret_Mounted_HV_Proj_By_Cue
root=Modulator_0
nodes [
 {id=Modulator_0, children [Random_0], properties {pitchMax=1.15}}
 {id=Random_0, children [WavePlayer_0,WavePlayer_1,WavePlayer_2,WavePlayer_3,WavePlayer_4], properties {weights [1,1,1,1,1]}}
 {id=WavePlayer_0, properties {soundWaveAssetPtr=&/WPN_Turret_Mounted_Heavy_Proj_By_02}}
 {id=WavePlayer_1, properties {soundWaveAssetPtr=&/WPN_Turret_Mounted_Heavy_Proj_By_03}}
 {id=WavePlayer_2, properties {soundWaveAssetPtr=&/WPN_Turret_Mounted_Heavy_Proj_By_04}}
 {id=WavePlayer_3, properties {soundWaveAssetPtr=&/WPN_Turret_Mounted_Heavy_Proj_By_05}}
 {id=WavePlayer_4, properties {soundWaveAssetPtr=&/WPN_Turret_Mounted_Heavy_Proj_By_01}}
]
properties {
 bOverrideAttenuation=true
 maxDistance=2097152
 soundClassObject=&Fort_WeaponRanged
 volumeMultiplier=1.5
 attenuationOverrides {bAttenuate=false, bAttenuateWithLPF=true, lPFFrequencyAtMax=1000, lPFRadiusMax=1500, lPFRadiusMin=500, stereoSpread=0}
 concurrencySet [/Game/Sounds/Concurrency_Rules/Athena_BulletWhizBy_Conc]
}
