SoundCue /Game/Sounds/Fort_Weapons/Turret/Turret_Mounted_HV_Proj_Water_Cue
root=Mixer_0
nodes [
 {
  id=Attenuation_0
  children [Modulator_3]
  properties {attenuationSettings=&Fort_Weapon_Projectile_Medium_Mono_Athena_Att}
 }
 {
  id=Attenuation_1
  children [Modulator_0]
  properties {attenuationSettings=&Fort_Weapon_Projectile_Distant_Mono_Athena_Att}
 }
 {
  id=Attenuation_3
  children [Modulator_2]
  properties {attenuationSettings=&Fort_Weapon_Projectile_Close_Athena_Att, attenuationOverrides {falloffDistance=1600}}
 }
 {id=Delay_0, children [Mixer_1], properties {delayMax=0.066, delayMin=0.066}}
 {id=Mixer_0, children [SoundClass_0,SoundClass_1,SoundClass_2], properties {inputVolume [1,1,1]}}
 {id=Mixer_1, children [Random_3,Random_0], properties {inputVolume [1,1]}}
 {id=Mixer_2, children [Random_5,Random_1], properties {inputVolume [1,1]}}
 {id=Modulator_0, children [Random_6], properties {pitchMax=1.1, pitchMin=0.9}}
 {id=Modulator_2, children [Delay_0], properties {pitchMax=1.1, pitchMin=1}}
 {id=Modulator_3, children [Mixer_2], properties {pitchMax=1.1, pitchMin=0.8}}
 {id=Random_0, children [WavePlayer_0,WavePlayer_1,WavePlayer_2,WavePlayer_4], properties {weights [1,1,1,1]}}
 {id=Random_1, children [WavePlayer_5,WavePlayer_6,WavePlayer_7], properties {weights [1,1,1]}}
 {id=Random_3, children [WavePlayer_3, WavePlayer_14, WavePlayer_15, WavePlayer_16], properties {weights [1,1,1,1]}}
 {id=Random_5, children [WavePlayer_17, WavePlayer_18, WavePlayer_19, WavePlayer_20], properties {weights [1,1,1,1]}}
 {id=Random_6, children [WavePlayer_21, WavePlayer_24, WavePlayer_25, WavePlayer_26], properties {weights [1,1,1,1]}}
 {
  id=SoundClass_0
  children [Attenuation_3]
  properties {soundClassOverride=&Fort_WeaponRanged_Fire_3P_PSM}
 }
 {
  id=SoundClass_1
  children [Attenuation_0]
  properties {soundClassOverride=&Fort_WeaponRanged_Fire_Distant}
 }
 {
  id=SoundClass_2
  children [Attenuation_1]
  properties {soundClassOverride=&Fort_WeaponRanged_Fire_Distant}
 }
 {id=WavePlayer_0, properties {soundWaveAssetPtr=/Game/Athena/Sounds/Weapons/Explosions/EXPL_Debris_Water_Burst_Close_04}}
 {id=WavePlayer_1, properties {soundWaveAssetPtr=/Game/Athena/Sounds/Weapons/Explosions/EXPL_Debris_Water_Burst_Close_03}}
 {id=WavePlayer_14, properties {soundWaveAssetPtr=&/WPN_Turret_Mounted_Heavy_Proj_Expl_Close_02}}
 {id=WavePlayer_15, properties {soundWaveAssetPtr=&/WPN_Turret_Mounted_Heavy_Proj_Expl_Close_03}}
 {id=WavePlayer_16, properties {soundWaveAssetPtr=&/WPN_Turret_Mounted_Heavy_Proj_Expl_Close_04}}
 {id=WavePlayer_17, properties {soundWaveAssetPtr=&/WPN_Turret_Mounted_Heavy_Proj_Expl_Med_04}}
 {id=WavePlayer_18, properties {soundWaveAssetPtr=&/WPN_Turret_Mounted_Heavy_Proj_Expl_Med_01}}
 {id=WavePlayer_19, properties {soundWaveAssetPtr=&/WPN_Turret_Mounted_Heavy_Proj_Expl_Med_02}}
 {id=WavePlayer_2, properties {soundWaveAssetPtr=/Game/Athena/Sounds/Weapons/Explosions/EXPL_Debris_Water_Burst_Close_02}}
 {id=WavePlayer_20, properties {soundWaveAssetPtr=&/WPN_Turret_Mounted_Heavy_Proj_Expl_Med_03}}
 {id=WavePlayer_21, properties {soundWaveAssetPtr=&/WPN_Turret_Mounted_Heavy_Proj_Expl_Dist_04}}
 {id=WavePlayer_24, properties {soundWaveAssetPtr=&/WPN_Turret_Mounted_Heavy_Proj_Expl_Dist_01}}
 {id=WavePlayer_25, properties {soundWaveAssetPtr=&/WPN_Turret_Mounted_Heavy_Proj_Expl_Dist_02}}
 {id=WavePlayer_26, properties {soundWaveAssetPtr=&/WPN_Turret_Mounted_Heavy_Proj_Expl_Dist_03}}
 {id=WavePlayer_3, properties {soundWaveAssetPtr=&/WPN_Turret_Mounted_Heavy_Proj_Expl_Close_01}}
 {id=WavePlayer_4, properties {soundWaveAssetPtr=/Game/Athena/Sounds/Weapons/Explosions/EXPL_Debris_Water_Burst_Close_01}}
 {id=WavePlayer_5, properties {soundWaveAssetPtr=/Game/Athena/Sounds/Weapons/Explosions/EXPL_Debris_Water_Burst_Distant_01}}
 {id=WavePlayer_6, properties {soundWaveAssetPtr=/Game/Athena/Sounds/Weapons/Explosions/EXPL_Debris_Water_Burst_Distant_02}}
 {id=WavePlayer_7, properties {soundWaveAssetPtr=/Game/Athena/Sounds/Weapons/Explosions/EXPL_Debris_Water_Burst_Distant_03}}
]
properties {bOverrideConcurrency=true, maxDistance=27500, volumeMultiplier=1, concurrencyOverrides {maxCount=2, resolutionRule=StopOldest}}
