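# Sound cue for the rotation of a mounted heavy turret, with audio that follows aim movement.
# A mixer layers two looping turn waves, one driven by vertical aim delta and the other by horizontal, with the horizontal layer weighted louder.
# Continuous modulators map each delta parameter to that loop's volume and pitch.
# The cue uses close-range weapon attenuation, the traps sound class and single-instance concurrency that prevents new plays.
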SoundCue /Game/Sounds/Fort_Weapons/Turret/Turret_Mounted_HV_Rotate_Cue
root=Mixer_0
nodes [
 {id=Mixer_0, children [ModulatorContinuous_1, ModulatorContinuous_0], properties {inputVolume [1,2]}}
 {
  id=ModulatorContinuous_0
  children [WavePlayer_2]
  properties {
   pitchModulationParams {default=0, minOutput=0.9, parameterName=HorizontalAimDelta}
   volumeModulationParams {default=0, parameterName=HorizontalAimDelta}
  }
 }
 {
  id=ModulatorContinuous_1
  children [WavePlayer_1]
  properties {
   pitchModulationParams {default=0, maxOutput=1.2, minOutput=1, parameterName=VerticalAimDelta}
   volumeModulationParams {default=0, parameterName=VerticalAimDelta}
  }
 }
 {id=WavePlayer_1, properties {bLooping=true, soundWaveAssetPtr=&/WPN_Turret_Mounted_Heavy_Turn_Loop_01}}
 {id=WavePlayer_2, properties {bLooping=true, soundWaveAssetPtr=&/WPN_Turret_Mounted_Heavy_Turn_Loop_02}}
]
properties {
 attenuationSettings=&Fort_Weapon_Projectile_Close_Athena_Att
 maxDistance=4950
 soundClassObject=&Fort_Traps
 concurrencyOverrides {maxCount=1, resolutionRule=PreventNew}
}
