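# Sound cue for a turret's single shot, built from reused sniper, rifle and lightning sounds pitched up cue-wide.
# The root mixer combines medium and distant branches, each with its own attenuation and sound class override.
# The medium layer mixes a random sniper shot with a pitched-up lightning sound; the distant layer slightly randomizes a rifle shot's pitch.
# Concurrency is limited per owner, stopping the oldest.
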SoundCue /Game/Sounds/Fort_Weapons/Turret/Weapon_Turret_Single_Fire_Cue
root=Mixer_3
nodes [
 {id=Attenuation_0, children [Modulator_0], properties {attenuationSettings=/Game/Sounds/Attenuation/Fort_Weapon_Projectile_Distant_StW_Att}}
 {id=Attenuation_34, children [Mixer_0], properties {attenuationSettings=/Game/Sounds/Attenuation/Fort_Weapon_Projectile_Medium_StW_Att}}
 {id=Mixer_0, children [Random_0,Modulator_5], properties {inputVolume [0.7,0.5]}}
 {id=Mixer_3, children [SoundClass_16, SoundClass_0], properties {inputVolume [1,1]}}
 {id=Modulator_0, children [WavePlayer_2], properties {pitchMax=1.1, pitchMin=0.9}}
 {id=Modulator_5, children [WavePlayer_1], properties {pitchMax=1.5, pitchMin=1.5}}
 {id=Random_0, children [WavePlayer_5,WavePlayer_6,WavePlayer_7], properties {weights [1,1,1]}}
 {
  id=SoundClass_0
  children [Attenuation_0]
  properties {soundClassOverride=&Fort_WeaponRanged_Fire_Distant}
 }
 {
  id=SoundClass_16
  children [Attenuation_34]
  properties {soundClassOverride=&Fort_WeaponRanged_Fire_3P_PSM}
 }
 {id=WavePlayer_1, properties {soundWaveAssetPtr=/Game/Sounds/Fort_Weapons/AMR/Elemental_LIGHTNING_Death_Large_006}}
 {id=WavePlayer_2, properties {soundWaveAssetPtr=/Game/Sounds/Fort_Weapons/Rifle/Rifle_AMR_Shot_Distant06c}}
 {id=WavePlayer_5, properties {soundWaveAssetPtr=/Game/Sounds/Fort_Weapons/SniperRifles/weapon_sniper_rifle_auto_fire_03}}
 {id=WavePlayer_6, properties {soundWaveAssetPtr=/Game/Sounds/Fort_Weapons/SniperRifles/weapon_sniper_rifle_auto_fire_01}}
 {id=WavePlayer_7, properties {soundWaveAssetPtr=/Game/Sounds/Fort_Weapons/SniperRifles/weapon_sniper_rifle_auto_fire_02}}
]
properties {
 bOverrideConcurrency=true
 maxDistance=20000
 pitchMultiplier=2
 volumeMultiplier=0.4
 attenuationOverrides {falloffDistance=4100, attenuationShapeExtents {x=900}}
 concurrencyOverrides {bLimitToOwner=true, maxCount=2, resolutionRule=StopOldest}
}
